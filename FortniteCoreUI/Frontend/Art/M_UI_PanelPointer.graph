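# UI translucent material that draws a procedural triangle, named as a panel pointer, with color and scale parameters.
# UVs scale uniformly about center by TriangleScale's red channel, then U and 1-U are each divided by 1-V to form diagonal edges.
# PoorManSmoothStep softens both edges by Sharpness, and their product is raised to a power.
# Static switch CornerWedge swaps this for a hard-edged wedge by flooring a single ratio.
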Material /FortniteCoreUI/Frontend/Art/M_UI_PanelPointer
graph {
 properties {
  bUsedWithStaticMesh=false
  blendMode=BLEND_Translucent
  materialDomain=MD_UI
  pixelDepthOffsetMode=PDOM_Legacy
  usedShadingModels=Default Lit
 }
 nodes [
  n1 = VectorParameter() {parameterName=TriangleScale, defaultValue=<rgba 1 1 0 1>}
  n2 = MaterialFunctionCall(UVs=TextureCoordinate(), Texture Scale=AppendVector(A=@n1.R, B=@n1.R)) {materialFunction=&ScaleUVsByCenter}
  n3 = MaterialFunctionCall(Float2=@n2.UVs) {materialFunction=&BreakOutFloat2Components}
  n4 = OneMinus(None=@n3.G)
  n5 = Divide(A=@n3.R, B=@n4)
  Add(A=@n5, B=Constant() {r=0.01})
  Constant()
  Constant() {r=0.1}
  Constant()
  n6 = MaterialFunctionCall(Float2=@n2.UVs) {materialFunction=&BreakOutFloat2Components}
  n7 = OneMinus(None=@n6.G)
  Divide(A=@n4, B=@n7)
  n8 = Divide(A=OneMinus(None=@n6.R), B=@n7)
  Floor(None=@n8)
  MaterialFunctionCall() {materialFunction=&DebugTimeSine}
  n9 = ScalarParameter() {defaultValue=1, parameterName=Sharpness}
  n10 = Multiply(
   A=MaterialFunctionCall!(threshold=@n5, distance=@n9) {materialFunction=/Game/UI/DamageText/Lightning/PoorManSmoothStep}.Result
   B=MaterialFunctionCall(threshold=@n8, distance=@n9) {materialFunction=&PoorManSmoothStep}.Result
  )
  OneMinus(None=@n10)
  Saturate(None=StaticSwitchParameter(True=Floor(None=@n5), False=Power(Base=@n10) {constExponent=8}) {parameterName=CornerWedge})
  VectorParameter() {parameterName=TriangleColor, defaultValue=<rgba 1 1 1 1>}
 ]
}
